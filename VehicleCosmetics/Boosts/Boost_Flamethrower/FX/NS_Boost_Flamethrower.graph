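# Niagara system for the Flamethrower vehicle boost, streaming flipbook flame sprites backward from exhaust sockets on the car's skeletal mesh.
# A short-lived blue core joins two near-identical larger flame emitters, one GPU and one CPU, with particles partitioned across filtered sockets.
# Painted boosts tint the outer flames per variant, otherwise a warm gradient applies; the core stays blue.
# ColorVariantID also indexes arrays for core spawn rate and outer texture contrast and channels.
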NiagaraSystem /VehicleCosmetics/Boosts/Boost_Flamethrower/FX/NS_Boost_Flamethrower
dependencies {
 usedDataInterfaces [
  &NiagaraDataInterfaceArrayFloat
  &NiagaraDataInterfaceSpriteRendererInfo
  &NiagaraDataInterfaceColorCurve
  &NiagaraDataInterfaceSkeletalMesh
 ]
 usedDynamicInputs []
 usedModules [
  &self:SystemSpawnScript.SystemScriptSource.SystemScriptGraph.NiagaraNodeAssignment_0.SetVariables_#1
  &SystemState
 ]
}
emitters {
 NS_Flame_Blue001 {
  emitterUpdateScript [
   &EmitterState()
   &VC_FilteredSocketCount()
   &SpawnRate(SpawnRate=200, Use Spawn Probability=false)
   &SpawnRate(
    SpawnGroup=1
    SpawnRate=&SelectFloatFromArray(
     Direct Array Index=$User.ColorVariantID
     Float Selection Array=DataInterface() {propertyValues {FloatData [50,50,50,50,50,40,50,50,50,100,200,50,50,50,50,50,50]}}
    )
    Use Spawn Probability=false
   ) {module=SpawnRate001}
  ]
  particleSpawnScript [
   &VC_PaintedVariantsOverride(
    Black=<rgba 0.1615 0.1615 1 1>
    BurntSienna=<rgba 0.1615 0.1615 1 1>
    Cobalt=<rgba 0.1615 0.1615 1 1>
    Crimson=<rgba 0.1615 0.1615 1 1>
    DefaultColor=<rgba 0.1615 0.1615 1 1>
    ForestGreen=<rgba 0.1615 0.1615 1 1>
    Grey=<rgba 0.1615 0.1615 1 1>
    Lime=<rgba 0.1615 0.1615 1 1>
    Orange=<rgba 0.1615 0.1615 1 1>
    Pink=<rgba 0.1615 0.1615 1 1>
    Purple=<rgba 0.1615 0.1615 1 1>
    Saffron=<rgba 0.1615 0.1615 1 1>
    SkyBlue=<rgba 0.1615 0.1615 1 1>
    TitaniumWhite=<rgba 0.1615 0.1615 1 1>
   )
   &InitializeParticle(
    Color Mode=::NewEnumerator0
    Lifetime=0.2
    Sprite Size Mode=::NewEnumerator3
    Uniform Sprite Size=105
    UsePositionOffset=false
    Write Material Random=true
   )
   &PartitionParticles(Number Of Partitions=$Emitter.PortCount)
   &SkeletalMeshLocation(
    Bone / Socket Index=$Particles.PartitionParticles.Partition
    Bone / Socket Sampling Mode=::NewEnumerator5
    MaskSpawnGroup=false
    Mesh Sampling Type=::NewEnumerator6
    Sampled Position Offset=<-50 0 0>
    Skeletal Mesh=$System.SkeletalMesh
   )
   &AddVelocity(
    Cone Angle=15
    Module.UseConeFalloff=false
    Module.UseVelDistribution=false
    Override Seed=false
    Velocity Mode=::NewEnumerator2
    Velocity Speed=-1500
   )
   /SetVariables(Particles.VisibilityTag=$Emitter.SpawnGroup) {isSetParameters=true}
  ]
  particleUpdateScript [
   &ParticleState()
   &ScaleSpriteSize(Uniform Curve Sprite Scale=DataInterface() {propertyValues {Curve {keys [{arriveTangent=0.1917, interpMode=RCIM_Cubic, leaveTangent=0.1917, tangentMode=RCTM_User, time=0, value=0.9}, <at 1 1 RCIM_Cubic>]}}})
   &SolveForcesAndVelocity(Clamp Velocity=false, Limit Acceleration=false)
   &SubUVAnimation(
    "Sprite Renderer "=DataInterface() {propertyValues {SpriteRenderer=&self:NS_Flame_Blue001_1.NiagaraSpriteRendererProperties_3}}
    SubUV Animation Mode:ENiagara_SubUVLookupModeV2=::NewEnumerator2
    UseEndFrame=false
    UseStartFrame=false
   )
   &Color(Color=&MakeCustomLinearColorFromBool(
     Bool=$User.bIsPainted
     True Color=&MakeLinearColorFromVectorAndFloat(
      "Vector (RGB)"=&MakeVectorFromLinearColorRGB(Color=$Particles.Color)
      "Float (Alpha)"=&FloatFromCurve(
       FloatCurve=DataInterface() {propertyValues {Curve {
          defaultValue=1
          keys [{interpMode=RCIM_Cubic, tangentMode=RCTM_User, time=0, value=1}, {interpMode=RCIM_Cubic, tangentMode=RCTM_User, time=1, value=0}]
         }}}
       Scale Curve=&RandomRangeFloat(Minimum=0.1, Maximum=0.3)
      )
     )
     False Color=&MakeLinearColorFromVectorAndFloat(
      "Vector (RGB)"=&MakeVectorFromLinearColorRGB(Color=<rgba 0.1615 0.1615 1 1>)
      "Float (Alpha)"=&FloatFromCurve(
       FloatCurve=DataInterface() {propertyValues {Curve {
          defaultValue=1
          keys [{interpMode=RCIM_Cubic, tangentMode=RCTM_User, time=0, value=1}, {interpMode=RCIM_Cubic, tangentMode=RCTM_User, time=1, value=0}]
         }}}
       Scale Curve=&RandomRangeFloat(Minimum=0.1, Maximum=0.3)
      )
     )
    ))
  ]
  rendererClasses [&NiagaraSpriteRendererProperties]
  renderers [
   {rendererClass=&NiagaraSpriteRendererProperties}
   {rendererClass=&NiagaraSpriteRendererProperties, rendererIndex=1}
  ]
 }
 NS_Flame_Red {
  simTarget=GPUComputeSim
  emitterUpdateScript [
   &EmitterState()
   &VC_FilteredSocketCount()
   &SpawnRate(SpawnRate=500, Use Spawn Probability=false)
   /SetVariables(
    Emitter.Contrast Tex2:NiagaraFloat=&SelectFloatFromArray(
     Direct Array Index=$User.ColorVariantID
     Float Selection Array=DataInterface() {propertyValues {FloatData [1,1,0.7,0.8,0.8,0.5,1,0.9,0.75,1,0.85,0.6,1,0.6,1,1]}}
    )
    Emitter.RGB Chanel Selector:LinearColor=&SelectLinearColorFromArray(
     Array Sampling Mode=::NewEnumerator1
     Direct Array Index=$User.ColorVariantID
     Color Selection Array=DataInterface() {propertyValues {ColorData [
        <rgba 5 0 0 0>
        <rgba 5 0 0 0>
        <rgba 5 0 0 0>
        <rgba 5 0 0 0>
        <rgba 5 0 0 0>
        <rgba 0.7354 0 0 0>
        <rgba 1 0 0 0>
        <rgba 5 0 0 0>
        <rgba 2.5781 0 0 0>
        <rgba 5 0 0 0>
        <rgba 5 0 0 0>
        <rgba 0.5 0 0 0>
        <rgba 5 0 0 0>
        <rgba 1 0 0 0>
        <rgba 5 0 0 0>
        <rgba 5 0 0 0>
       ]}}
     Random Seed=?
    )
    Emitter.RGB Chanel Selector Texture 2:LinearColor=&SelectLinearColorFromArray(
     Array Sampling Mode=::NewEnumerator1
     Direct Array Index=$User.ColorVariantID
     Color Selection Array=DataInterface() {propertyValues {ColorData [
        <rgba 5 0 0 0>
        <rgba 5 0 0 0>
        <rgba 5 0 0 0>
        <rgba 5 0 0 0>
        <rgba 5 0 0 0>
        <rgba 5 0 0 0>
        <rgba 5 0 0 0>
        <rgba 5 0 0 0>
        <rgba 5 0 0 0>
        <rgba 5 0 0 0>
        <rgba 5 0 0 0>
        <rgba 20 0 0 0>
        <rgba 5 0 0 0>
        <rgba 5 0 0 0>
        <rgba 5 0 0 0>
        <rgba 5 0 0 0>
       ]}}
     Random Seed=?
    )
   ) {isSetParameters=true, module=SetVariables_#2}
  ]
  particleSpawnScript [
   &VC_PaintedVariantsOverride(
    Black=<rgba 0.3 0.3 0.3 1>
    BurntSienna=<rgba 0.7969 0.3344 0.018 1>
    Cobalt=<rgba 0 0.037 2 1>
    Crimson=<rgba 1.3391 0 0 1>
    ForestGreen=<rgba 0 1.25 0.0632 1>
    Grey=<rgba 0.6753 0.6753 0.6753 1>
    Lime=<rgba 0.7259 1.25 0 1>
    Orange=<rgba 1.25 0.738 0 1>
    Pink=<rgba 1.25 0.6228 1.0219 1>
    Purple=<rgba 0.874 0 1.25 1>
    Saffron=<rgba 1 0.7825 0.0677 1>
    SkyBlue=<rgba 0 1.3308 2 1>
    TitaniumWhite=<rgba 1.25 1.25 1.25 1>
   )
   &InitializeParticle(
    Color Mode=::NewEnumerator0
    Lifetime=0.5
    Sprite Rotation Mode=::NewEnumerator2
    Sprite Size Mode=::NewEnumerator3
    Uniform Sprite Size=125
    UsePositionOffset=false
    Write Material Random=true
   )
   &PartitionParticles(Number Of Partitions=$Emitter.PortCount)
   &SkeletalMeshLocation(
    Bone / Socket Index=$Particles.PartitionParticles.Partition
    Bone / Socket Sampling Mode=::NewEnumerator5
    MaskSpawnGroup=false
    Mesh Sampling Type=::NewEnumerator6
    Sampled Position Offset=&AddVectorToPosition(Position=&RandomRangeVector(Minimum=<-300 0 0>, Maximum=<0 1 0>), Vector=$User.DTCarOffset)
    Skeletal Mesh=$System.SkeletalMesh
   )
   &AddVelocity(Velocity=<-1500 0 0>)
  ]
  particleUpdateScript [
   &Color(Color=&MakeCustomLinearColorFromBool(
     Bool=$User.bIsPainted
     True Color=&Multiply_LinearColorByFloat(
      Linear Color=&MultiplyLinearColor(
       Linear Color=$Particles.Color
       Scale Factor=&MakeVector4FromVectorAndFloat(
        "Vector (RGB)"=&VectorFromFloat(Value=&FloatFromCurve(FloatCurve=DataInterface() {propertyValues {Curve {keys [<at 0.2 1 RCIM_Cubic>,<at 1 0.8 RCIM_Cubic>]}}}))
        "Float (Alpha)"=&FloatFromCurve(FloatCurve=DataInterface() {propertyValues {Curve {keys [<at 0.6 1 RCIM_Cubic>, {arriveTangent=-6.0312, interpMode=RCIM_Cubic, leaveTangent=-6.0312, tangentMode=RCTM_User, time=1.2012, value=0}]}}})
       )
      )
      Float=1
     )
     False Color=&Multiply_LinearColorByFloat(
      Linear Color=&ColorFromCurve(ColorCurve=DataInterface() {propertyValues {
         AlphaCurve {keys [<at 0 0.1>, {arriveTangent=-0.1051, interpMode=RCIM_Cubic, time=0.259, value=1}, <at 1 0 RCIM_Cubic>]}
         BlueCurve {keys [
           {arriveTangent=-0.8592, interpMode=RCIM_Cubic, time=0.2051, value=0.4102}
           {arriveTangent=-0.3763, interpMode=RCIM_Cubic, leaveTangent=-0.3763, time=0.7072, value=0.25}
           <at 0.8559 0.1653>
          ]}
         GreenCurve {keys [
           {arriveTangent=-1.3862, interpMode=RCIM_Cubic, time=0.2051, value=0.8055}
           {arriveTangent=-0.647, interpMode=RCIM_Cubic, leaveTangent=-0.647, time=0.7072, value=0.4585}
           <at 0.8559 0.3845>
          ]}
         RedCurve {keys [
           {arriveTangent=-1.045, interpMode=RCIM_Cubic, time=0.2051, value=1.25}
           {arriveTangent=-0.3842, interpMode=RCIM_Cubic, leaveTangent=-0.3842, time=0.7072, value=1}
           <at 0.8559 1>
          ]}
        }})
      Float=1
     )
    ))
   &ParticleState()
   &ScaleSpriteSize(Uniform Curve Sprite Scale=DataInterface() {propertyValues {Curve {keys [
        {arriveTangent=0.1117, interpMode=RCIM_Cubic, leaveTangent=0.1117, tangentMode=RCTM_User, time=0, value=0.6}
        {arriveTangent=1.3406, interpMode=RCIM_Cubic, leaveTangent=1.9104, tangentMode=RCTM_Break, time=0.3972, value=0.7467}
        <at 1 1.3231 RCIM_Cubic>
       ]}}})
   &SolveForcesAndVelocity(Clamp Velocity=false, Limit Acceleration=false)
   &SubUVAnimation(
    "Sprite Renderer "=DataInterface() {propertyValues {SpriteRenderer=&self:NS_Flame_Red_0.NiagaraSpriteRendererProperties_3}}
    SubUV Animation Mode:ENiagara_SubUVLookupModeV2=::NewEnumerator2
    UseEndFrame=false
    UseStartFrame=false
   )
  ]
  rendererClasses [&NiagaraSpriteRendererProperties]
  renderers [{rendererClass=&NiagaraSpriteRendererProperties}]
 }
 NS_Flame_Red_01 {
  emitterUpdateScript [
   &EmitterState()
   &VC_FilteredSocketCount()
   &SpawnRate(SpawnRate=500, Use Spawn Probability=false)
   /SetVariables(
    Emitter.Contrast Tex2:NiagaraFloat=&SelectFloatFromArray(
     Direct Array Index=$User.ColorVariantID
     Float Selection Array=DataInterface() {propertyValues {FloatData [1,1,0.7,0.8,0.8,0.5,1,0.9,0.75,1,0.85,0.6,1,0.6,1,1]}}
    )
    Emitter.RGB Chanel Selector:LinearColor=&SelectLinearColorFromArray(
     Array Sampling Mode=::NewEnumerator1
     Direct Array Index=$User.ColorVariantID
     Color Selection Array=DataInterface() {propertyValues {ColorData [
        <rgba 5 0 0 0>
        <rgba 5 0 0 0>
        <rgba 5 0 0 0>
        <rgba 5 0 0 0>
        <rgba 5 0 0 0>
        <rgba 0.7354 0 0 0>
        <rgba 1 0 0 0>
        <rgba 5 0 0 0>
        <rgba 2.5781 0 0 0>
        <rgba 5 0 0 0>
        <rgba 5 0 0 0>
        <rgba 0.5 0 0 0>
        <rgba 5 0 0 0>
        <rgba 1 0 0 0>
        <rgba 5 0 0 0>
        <rgba 5 0 0 0>
       ]}}
     Random Seed=?
    )
    Emitter.RGB Chanel Selector Texture 2:LinearColor=&SelectLinearColorFromArray(
     Array Sampling Mode=::NewEnumerator1
     Direct Array Index=$User.ColorVariantID
     Color Selection Array=DataInterface() {propertyValues {ColorData [
        <rgba 5 0 0 0>
        <rgba 5 0 0 0>
        <rgba 5 0 0 0>
        <rgba 5 0 0 0>
        <rgba 5 0 0 0>
        <rgba 5 0 0 0>
        <rgba 5 0 0 0>
        <rgba 5 0 0 0>
        <rgba 5 0 0 0>
        <rgba 5 0 0 0>
        <rgba 5 0 0 0>
        <rgba 20 0 0 0>
        <rgba 5 0 0 0>
        <rgba 5 0 0 0>
        <rgba 5 0 0 0>
        <rgba 5 0 0 0>
       ]}}
     Random Seed=?
    )
   ) {isSetParameters=true, module=SetVariables_#2}
  ]
  particleSpawnScript [
   &VC_PaintedVariantsOverride(
    Black=<rgba 0.3 0.3 0.3 1>
    BurntSienna=<rgba 0.7969 0.3344 0.018 1>
    Cobalt=<rgba 0 0.037 2 1>
    Crimson=<rgba 1.3391 0 0 1>
    ForestGreen=<rgba 0 1.25 0.0632 1>
    Grey=<rgba 0.6753 0.6753 0.6753 1>
    Lime=<rgba 0.7259 1.25 0 1>
    Orange=<rgba 1.25 0.738 0 1>
    Pink=<rgba 1.25 0.6228 1.0219 1>
    Purple=<rgba 0.874 0 1.25 1>
    Saffron=<rgba 1 0.7825 0.0677 1>
    SkyBlue=<rgba 0 1.3308 2 1>
    TitaniumWhite=<rgba 1.25 1.25 1.25 1>
   )
   &InitializeParticle(
    Color Mode=::NewEnumerator0
    Lifetime=0.5
    Sprite Rotation Mode=::NewEnumerator2
    Sprite Size Mode=::NewEnumerator3
    Uniform Sprite Size=125
    UsePositionOffset=false
    Write Material Random=true
   )
   &PartitionParticles(Number Of Partitions=$Emitter.PortCount)
   &SkeletalMeshLocation(
    Bone / Socket Index=$Particles.PartitionParticles.Partition
    Bone / Socket Sampling Mode=::NewEnumerator5
    MaskSpawnGroup=false
    Mesh Sampling Type=::NewEnumerator6
    Sampled Position Offset=&AddVectorToPosition(Position=&RandomRangeVector(Minimum=<-300 0 0>, Maximum=<0 1 0>), Vector=$User.DTCarOffset)
    Skeletal Mesh=$System.SkeletalMesh
   )
   &AddVelocity(Velocity=<-1500 0 0>)
  ]
  particleUpdateScript [
   &Color(Color=&MakeCustomLinearColorFromBool(
     Bool=$User.bIsPainted
     True Color=&Multiply_LinearColorByFloat(
      Linear Color=&MultiplyLinearColor(
       Linear Color=$Particles.Color
       Scale Factor=&MakeVector4FromVectorAndFloat(
        "Vector (RGB)"=&VectorFromFloat(Value=&FloatFromCurve(FloatCurve=DataInterface() {propertyValues {Curve {keys [<at 0.2 1 RCIM_Cubic>,<at 1 0.8 RCIM_Cubic>]}}}))
        "Float (Alpha)"=&FloatFromCurve(FloatCurve=DataInterface() {propertyValues {Curve {keys [<at 0.6 1 RCIM_Cubic>, {arriveTangent=-6.0312, interpMode=RCIM_Cubic, leaveTangent=-6.0312, tangentMode=RCTM_User, time=1.2012, value=0}]}}})
       )
      )
      Float=1
     )
     False Color=&Multiply_LinearColorByFloat(
      Linear Color=&ColorFromCurve(ColorCurve=DataInterface() {propertyValues {
         AlphaCurve {keys [<at 0 0.1>, {arriveTangent=-0.1051, interpMode=RCIM_Cubic, time=0.259, value=1}, <at 1 0 RCIM_Cubic>]}
         BlueCurve {keys [
           {arriveTangent=-0.8592, interpMode=RCIM_Cubic, time=0.2051, value=0.4102}
           {arriveTangent=-0.3763, interpMode=RCIM_Cubic, leaveTangent=-0.3763, time=0.7072, value=0.25}
           <at 0.8559 0.1653>
          ]}
         GreenCurve {keys [
           {arriveTangent=-1.3862, interpMode=RCIM_Cubic, time=0.2051, value=0.8055}
           {arriveTangent=-0.647, interpMode=RCIM_Cubic, leaveTangent=-0.647, time=0.7072, value=0.4585}
           <at 0.8559 0.3845>
          ]}
         RedCurve {keys [
           {arriveTangent=-1.045, interpMode=RCIM_Cubic, time=0.2051, value=1.25}
           {arriveTangent=-0.3842, interpMode=RCIM_Cubic, leaveTangent=-0.3842, time=0.7072, value=1}
           <at 0.8559 1>
          ]}
        }})
      Float=1
     )
    ))
   &ParticleState()
   &ScaleSpriteSize(Uniform Curve Sprite Scale=DataInterface() {propertyValues {Curve {keys [
        {arriveTangent=0.1117, interpMode=RCIM_Cubic, leaveTangent=0.1117, tangentMode=RCTM_User, time=0, value=0.6}
        {arriveTangent=1.3406, interpMode=RCIM_Cubic, leaveTangent=1.9104, tangentMode=RCTM_Break, time=0.3972, value=0.7467}
        <at 1 1.3231 RCIM_Cubic>
       ]}}})
   &SolveForcesAndVelocity(Clamp Velocity=false, Limit Acceleration=false)
   &SubUVAnimation(
    "Sprite Renderer "=DataInterface() {propertyValues {SpriteRenderer=&self:NS_Flame_Red_01_0.NiagaraSpriteRendererProperties_3}}
    SubUV Animation Mode:ENiagara_SubUVLookupModeV2=::NewEnumerator2
    UseEndFrame=false
    UseStartFrame=false
   )
  ]
  rendererClasses [&NiagaraSpriteRendererProperties]
  renderers [{rendererClass=&NiagaraSpriteRendererProperties}]
 }
}
summary {
 systemName=NS_Boost_Flamethrower
 emitters [
  {emitterName=NS_Flame_Blue001, rendererClasses [&NiagaraSpriteRendererProperties]}
  {emitterName=NS_Flame_Red, simTarget=GPUComputeSim, rendererClasses [&NiagaraSpriteRendererProperties]}
  {emitterName=NS_Flame_Red_01, rendererClasses [&NiagaraSpriteRendererProperties]}
 ]
 userVariables [
  User.bIsFrontendPreview {type=NiagaraBool}
  User.bIsPainted {type=NiagaraBool}
  User.ColorVariantID {type=NiagaraInt32}
  User.DTCarOffset {type=Vector3f, description="", defaultValue {struct=Vector3f, value=<0 0 0>}}
  User.Emitter_Color Array {
   type=NiagaraDataInterfaceArrayColor
   description=""
   defaultValue {
    struct=NiagaraExt_VariableValue_DataInterface
    value {
     dataInterface=&self:NiagaraDataInterfaceArrayColor_0
     dataInterfaceClass=&NiagaraDataInterfaceArrayColor
    }
   }
  }
  User.GameModeID {type=NiagaraInt32}
  User.General Opacity Color {
   type=NiagaraDataInterfaceCurve
   description=""
   defaultValue {
    struct=NiagaraExt_VariableValue_DataInterface
    value {
     dataInterface=&self:NiagaraDataInterfaceCurve_0
     dataInterfaceClass=&NiagaraDataInterfaceCurve
    }
   }
  }
  User.PaintedColor {type=LinearColor, description="", defaultValue {struct=LinearColor, value=<rgba 1 1 1 1>}}
  User.PaintedScalar {type=NiagaraFloat, defaultValue=1, description=""}
  User.PortCount {type=NiagaraInt32, defaultValue=1}
 ]
}
